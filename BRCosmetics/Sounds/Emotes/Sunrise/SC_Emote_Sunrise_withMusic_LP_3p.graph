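# Third-person looping SoundCue for the Sunrise emote, mixing ambience, music and randomized worm sounds.
# The ambience and music loops fade in through short envelopes, with the music behind a licensed content switcher whose other inputs play looping silence.
# Two looping branches each wait a random delay before picking from their own set of worm variations.
# It uses emote attenuation, music concurrency and sound class, and sends to the music audio-analysis submix.
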SoundCue /BRCosmetics/Sounds/Emotes/Sunrise/SC_Emote_Sunrise_withMusic_LP_3p
root=Mixer_0
nodes [
 {id=Delay_0, children [Random_0], properties {delayMax=6, delayMin=2}}
 {id=Delay_1, children [Random_1], properties {delayMax=6, delayMin=2}}
 {
  id=Enveloper_0
  children [WavePlayer_0]
  properties {pitchCurve {editorCurveData {keys [{time=0}]}}, volumeCurve {editorCurveData {keys [<at 0 0>,<at 0.2 1>]}}}
 }
 {
  id=Enveloper_1
  children [WavePlayer_2]
  properties {pitchCurve {editorCurveData {keys [{time=0}]}}, volumeCurve {editorCurveData {keys [<at 0 0>,<at 0.25 1>]}}}
 }
 {id=FortSoundNodeLicensedContentSwitcher_0, class=FortSoundNodeLicensedContentSwitcher, children [WavePlayer_17, WavePlayer_17, Enveloper_0]}
 {id=Looping_0, children [Delay_0]}
 {id=Looping_1, children [Delay_1]}
 {id=Mixer_0, children [Enveloper_1, FortSoundNodeLicensedContentSwitcher_0, Looping_0, Looping_1], properties {inputVolume [1,0.9,1,1]}}
 {
  id=Random_0
  children [WavePlayer_3, WavePlayer_4, WavePlayer_5, WavePlayer_6, WavePlayer_7, WavePlayer_8, WavePlayer_10, WavePlayer_11]
  properties {weights [1,1,1,1,1,1,1,1]}
 }
 {
  id=Random_1
  children [WavePlayer_1, WavePlayer_12, WavePlayer_13, WavePlayer_14, WavePlayer_15, WavePlayer_16]
  properties {weights [1,1,1,1,1,1]}
 }
 {id=WavePlayer_0, properties {bLooping=true, soundWaveAssetPtr=&/Emote_Sunrise_MX_LP_03}}
 {id=WavePlayer_1, properties {soundWaveAssetPtr=&/Emote_Sunrise_Worm_02_01}}
 {id=WavePlayer_10, properties {soundWaveAssetPtr=&/Emote_Sunrise_Worm_08}}
 {id=WavePlayer_11, properties {soundWaveAssetPtr=&/Emote_Sunrise_Worm_09}}
 {id=WavePlayer_12, properties {soundWaveAssetPtr=&/Emote_Sunrise_Worm_02_02}}
 {id=WavePlayer_13, properties {soundWaveAssetPtr=&/Emote_Sunrise_Worm_02_03}}
 {id=WavePlayer_14, properties {soundWaveAssetPtr=&/Emote_Sunrise_Worm_02_04}}
 {id=WavePlayer_15, properties {soundWaveAssetPtr=&/Emote_Sunrise_Worm_02_05}}
 {id=WavePlayer_16, properties {soundWaveAssetPtr=&/Emote_Sunrise_Worm_02_06}}
 {id=WavePlayer_17, properties {bLooping=true, soundWaveAssetPtr=/Game/Sounds/Fort_Test_Sounds/5_seconds_silence}}
 {id=WavePlayer_2, properties {bLooping=true, soundWaveAssetPtr=&/Emote_Sunrise_Amb_LP}}
 {id=WavePlayer_3, properties {soundWaveAssetPtr=&/Emote_Sunrise_Worm_01}}
 {id=WavePlayer_4, properties {soundWaveAssetPtr=&/Emote_Sunrise_Worm_02}}
 {id=WavePlayer_5, properties {soundWaveAssetPtr=&/Emote_Sunrise_Worm_03}}
 {id=WavePlayer_6, properties {soundWaveAssetPtr=&/Emote_Sunrise_Worm_04}}
 {id=WavePlayer_7, properties {soundWaveAssetPtr=&/Emote_Sunrise_Worm_05}}
 {id=WavePlayer_8, properties {soundWaveAssetPtr=&/Emote_Sunrise_Worm_06}}
]
properties {
 attenuationSettings=&Fort_Emotes_Athena_Att
 audioPropertiesSheet=/BRCosmetics/Sounds/Settings/PropertySheets/Emotes/APS_Cosmetic_Emote_Music_3P
 maxDistance=1600
 soundClassObject=/Game/Sounds/FortSoundClassesAndModes/Fort_Music_Emote_3P
 virtualizationMode=PlayWhenSilent
 volumeMultiplier=1
 concurrencySet [/Game/Sounds/Concurrency_Rules/Athena_EmoteMusic_Conc]
 soundSubmixSends [{minSendLevel=0, soundSubmix=&Fortnite_AudioAnalysis_Music}]
}
